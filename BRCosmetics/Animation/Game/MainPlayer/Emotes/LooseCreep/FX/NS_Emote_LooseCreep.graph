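# Niagara system for the LooseCreep emote: a mostly blue blink effect with body trails, glows and a shockwave.
# Two anim-trail ribbon emitters draw streaks, the first layering five renderers with separate color and width bindings.
# Bursts add socket-placed lights, bone-attached sprites and a two-layer mesh shockwave.
# User.BlinkIndex selects per-blink lifetimes, counts, brightness and sizes, BlinkWidth scales trail width, and time-of-day scalars adjust trail and shockwave intensity.
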NiagaraSystem /BRCosmetics/Animation/Game/MainPlayer/Emotes/LooseCreep/FX/NS_Emote_LooseCreep
dependencies {
 usedDataInterfaces [
  &NiagaraDataInterfaceColorCurve
  &NiagaraDataInterfaceCurve
  &NiagaraDataInterfaceVectorCurve
  &NiagaraDataInterfaceSocketReader
 ]
 usedDynamicInputs []
 usedModules [
  &SystemState
  &self:SystemSpawnScript.SystemScriptSource.SystemScriptGraph.NiagaraNodeAssignment_1.SetVariables_#1
 ]
}
emitters {
 FN_AnimTrail001 {
  emitterSpawnScript [&FN_RibbonIDOffsetUtility()]
  emitterUpdateScript [
   &EmitterState(Life Cycle Mode=::NewEnumerator1, Loop Behavior=::NewEnumerator1, Loop Duration=0.2, UseLoopDelay=false)
   &FN_SpawnAnimTrail()
  ]
  particleSpawnScript [
   &FN_InitializeAnimTrailBoneIndecies()
   &FN_InitializeAnimTrail(
    AnimTrailWidth=&Multiply_Float(A=4, B=$User.BlinkWidth)
    Color=<rgba 1 1 1 0.45>
    Lifetime=&Add_Float(
     A=&MergeHelper_FloatSwitch(0=0.28, 1=0.28, 2=0.31, 3=0.23, Index=&MakeIntFromFloat(Float=$User.BlinkIndex), Num=::NewEnumerator2)
     B=0
    )
   )
   &PartitionParticles() {module=PartitionParticles001}
  ]
  particleUpdateScript [
   &SolveForcesAndVelocity(Clamp Velocity=false, Limit Acceleration=false)
   &FN_UpdateAnimTrail() {module=FN_UpdateAnimTrail001}
   &ScaleColor(
    Scale Alpha=&Multiply_Float(
     A=&FloatFromCurve(
      FloatCurve=DataInterface() {propertyValues {Curve {keys [
          <at 0 1>
          {arriveTangent=-0.0395, interpMode=RCIM_Cubic, leaveTangent=-0.0395, tangentMode=RCTM_User, time=0.25, value=1}
          {arriveTangent=-2.2142, interpMode=RCIM_Cubic, leaveTangent=-2.2142, tangentMode=RCTM_User, time=1, value=0}
         ]}}}
      Scale Curve=0.7
     )
     B=1
    )
    ScaleA=true
    ScaleRGB=true
   )
   &ScaleRibbonWidth(Ribbon Width Scale=0.95)
   /SetVariables(
    Particles.Color2=&Multiply_LinearColorByFloat(
     Linear Color=<rgba 4.1074 17.6745 25 1>
     Float=&Multiply_Float(
      A=&FloatFromCurve(FloatCurve=DataInterface() {propertyValues {Curve {keys [
           <at 0 1>
           {arriveTangent=-0.0395, interpMode=RCIM_Cubic, leaveTangent=-0.0395, tangentMode=RCTM_User, time=0.25, value=1}
           {arriveTangent=-1.6584, interpMode=RCIM_Cubic, leaveTangent=-1.6584, tangentMode=RCTM_User, time=1, value=0}
          ]}}})
      B=1
     )
    )
    Particles.Dynamic2:Vector4f=<3 1.45 0 0>
    Particles.RibbonWidth2:NiagaraFloat=&Multiply_Float(A=$Particles.RibbonWidth, B=0.85)
   ) {isSetParameters=true}
   /SetVariables(
    Particles.Color3:LinearColor=&Multiply_LinearColorByFloat(
     Linear Color=<rgba 1.1957 3.8257 10 1>
     Float=&Multiply_Float(
      A=&FloatFromCurve(FloatCurve=DataInterface() {propertyValues {Curve {keys [
           <at 0 1>
           {arriveTangent=-0.0395, interpMode=RCIM_Cubic, leaveTangent=-0.0395, tangentMode=RCTM_User, time=0.25, value=1}
           {arriveTangent=-1.6562, interpMode=RCIM_Cubic, leaveTangent=-1.6562, tangentMode=RCTM_User, time=1, value=0}
          ]}}})
      B=1
     )
    )
    Particles.Dynamic3:Vector4f=<1 1 0 0>
    Particles.RibbonWidth3:NiagaraFloat=&Multiply_Float(A=$Particles.RibbonWidth, B=0.875)
   ) {isSetParameters=true}
   /SetVariables(
    Particles.Color4:LinearColor=&Multiply_LinearColorByFloat(
     Linear Color=<rgba 3.1385 18.9109 50 1>
     Float=&Multiply_Float(
      A=&FloatFromCurve(
       FloatCurve=DataInterface() {propertyValues {Curve {keys [
           <at 0 1>
           {arriveTangent=-0.0395, interpMode=RCIM_Cubic, leaveTangent=-0.0395, tangentMode=RCTM_User, time=0.25, value=1}
           {arriveTangent=-1.6562, interpMode=RCIM_Cubic, leaveTangent=-1.6562, tangentMode=RCTM_User, time=1, value=0}
          ]}}}
       Scale Curve=0.2
      )
      B=1
     )
    )
    Particles.Dynamic4:Vector4f=&MakeVector4(
     X=&FloatFromCurve(FloatCurve=DataInterface() {propertyValues {Curve {keys [<at 0 1>,<at 1 0>]}}})
     Y=0
     Z=0
     W=1
    )
    Particles.RibbonWidth4:NiagaraFloat=&Multiply_Float(A=$Particles.RibbonWidth, B=1.3)
   ) {isSetParameters=true}
   /SetVariables(
    Particles.Color5:LinearColor=&Multiply_LinearColorByFloat(
     Linear Color=<rgba 2.1787 15.6433 25 1>
     Float=&Multiply_Float(
      A=&FloatFromCurve(FloatCurve=DataInterface() {propertyValues {Curve {keys [
           <at 0 1>
           {arriveTangent=-0.0395, interpMode=RCIM_Cubic, leaveTangent=-0.0395, tangentMode=RCTM_User, time=0.25, value=1}
           {arriveTangent=-1.6562, interpMode=RCIM_Cubic, leaveTangent=-1.6562, tangentMode=RCTM_User, time=1, value=0}
          ]}}})
      B=1
     )
    )
    Particles.Dynamic5:Vector4f=&MakeVector4(
     X=&FloatFromCurve(FloatCurve=DataInterface() {propertyValues {Curve {keys [<at 0 1>,<at 1 0>]}}})
     Y=0
     Z=0
     W=1
    )
    Particles.RibbonWidth5:NiagaraFloat=&Multiply_Float(A=$Particles.RibbonWidth, B=1.25)
   ) {isSetParameters=true}
   /SetVariables(
    Particles.Color6:LinearColor=&Multiply_LinearColorByFloat(
     Linear Color=&ColorFromCurve(ColorCurve=DataInterface() {propertyValues {
        AlphaCurve {keys [<at 0 1 RCIM_Cubic>,<at 1 0 RCIM_Cubic>]}
        BlueCurve {keys [
          <at 0 25 RCIM_Cubic>
          <at 0.1119 17>
          {arriveTangent=-13.5118, interpMode=RCIM_Cubic, leaveTangent=-13.5118, time=0.2248, value=15}
          <at 1 5>
         ]}
        GreenCurve {keys [
          <at 0 12.9871 RCIM_Cubic>
          <at 0.1119 9.9511>
          {arriveTangent=-8.3358, interpMode=RCIM_Cubic, leaveTangent=-8.3358, time=0.2248, value=7.074}
          <at 1 2.548>
         ]}
        RedCurve {keys [
          <at 0 1.2461 RCIM_Cubic>
          <at 0.1119 0.4235>
          {arriveTangent=-0.3218, interpMode=RCIM_Cubic, leaveTangent=-0.3218, time=0.2248, value=0.527}
          <at 1 0.1376>
         ]}
       }})
     Float=&Multiply_Float(
      A=&FloatFromCurve(
       FloatCurve=DataInterface() {propertyValues {Curve {keys [<at 0 1 RCIM_Cubic>, {arriveTangent=-0.0104, interpMode=RCIM_Cubic, leaveTangent=-0.0104, tangentMode=RCTM_User, time=0.8, value=0}]}}}
       Scale Curve=&FNTimeOfDayScalar(Night Scale=1.5, Morning Scale=1.25, Day Scale=2.5, Evening Scale=1.5)
      )
      B=1
     )
    )
    Particles.Dynamic6:Vector4f=<0 0 0 0>
    Particles.RibbonWidth6:NiagaraFloat=&Multiply_Float(
     A=$Particles.RibbonWidth
     B=&FloatFromCurve(
      FloatCurve=DataInterface() {propertyValues {Curve {keys [
          <at 0 0.8 RCIM_Cubic>
          {arriveTangent=-0.2209, interpMode=RCIM_Cubic, leaveTangent=-0.2209, tangentMode=RCTM_User, time=0.275, value=1}
          <at 1 0 RCIM_Cubic>
         ]}}}
      Scale Curve=1.25
     )
    )
   ) {isSetParameters=true, module=SetVariables_#2}
  ]
  rendererClasses [&NiagaraRibbonRendererProperties]
  renderers [
   {rendererClass=&NiagaraRibbonRendererProperties}
   {rendererClass=&NiagaraRibbonRendererProperties, rendererIndex=1}
   {rendererClass=&NiagaraRibbonRendererProperties, rendererIndex=2}
   {rendererClass=&NiagaraRibbonRendererProperties, rendererIndex=3}
   {rendererClass=&NiagaraRibbonRendererProperties, rendererIndex=4}
  ]
 }
 FN_AnimTrail002 {
  emitterSpawnScript [&FN_RibbonIDOffsetUtility()]
  emitterUpdateScript [
   &EmitterState()
   &FN_SpawnAnimTrail()
  ]
  particleSpawnScript [
   &FN_InitializeAnimTrailBoneIndecies()
   &FN_InitializeAnimTrailBoneIndecies(AnimTrail Selection=1, First Bone or Socket Index=2, Second Bone or Socket Index=3) {enabled=false, module=FN_InitializeAnimTrailBoneIndecies001}
   &FN_InitializeAnimTrail(
    AnimTrailWidth=&Multiply_Float(A=3.5, B=$User.BlinkWidth)
    Color=<rgba 1 1 1 0.1>
    Lifetime=&Add_Float(
     A=&MergeHelper_FloatSwitch(0=0.28, 1=0.28, 2=0.31, 3=0.23, Index=&MakeIntFromFloat(Float=$User.BlinkIndex), Num=::NewEnumerator2)
     B=0
    )
    Position Offset=<0 0 2>
   )
   &AddVelocity(Coordinate Space=::Simulation, Velocity=<0 -625 0>) {enabled=false}
   &AddVelocity(Velocity=&MakeVector(
     X=0
     Y=&FN_FrontEndFloatModifier(
      Gameplay=&Lerp_Float(
       A=0
       B=-625
       Alpha=&FloatFromCurve(
        FloatCurve=DataInterface() {propertyValues {Curve {keys [<at 0 0 RCIM_Constant>,<at 500 1 RCIM_Constant>]}}}
        CurveIndex=&VectorLength(Vector=$Engine.Owner.Velocity)
       )
      )
      Front End=-425
     )
     Z=0
    )) {enabled=false, module=AddVelocity001}
  ]
  particleUpdateScript [
   &ScaleVelocity(Velocity Scale=&MakeVector(
     X=0
     Y=&FloatFromCurve(FloatCurve=DataInterface() {propertyValues {Curve {keys [<at 0 1 RCIM_Cubic>,<at 0.15 0 RCIM_Cubic>]}}})
     Z=0
    )) {enabled=false}
   &SolveForcesAndVelocity(Clamp Velocity=false, Limit Acceleration=false)
   &FN_UpdateAnimTrail() {module=FN_UpdateAnimTrail001}
   /SetVariables(
    Particles.Color6:LinearColor=&Multiply_LinearColorByFloat(
     Linear Color=<rgba 0.1395 0.4426 1 0.08>
     Float=&Multiply_Float(
      A=&FloatFromCurve(
       FloatCurve=DataInterface() {propertyValues {Curve {keys [<at 0 1 RCIM_Cubic>, {arriveTangent=-2.5155, interpMode=RCIM_Cubic, leaveTangent=-2.5155, tangentMode=RCTM_User, time=1, value=0}]}}}
       Scale Curve=&FNTimeOfDayScalar(Night Scale=1.5, Morning Scale=1.25, Day Scale=2.5, Evening Scale=1.5)
      )
      B=1
     )
    )
    Particles.RibbonWidth6:NiagaraFloat=&Multiply_Float(A=$Particles.RibbonWidth, B=0.95)
   ) {isSetParameters=true, module=SetVariables_#2}
   &DynamicMaterialParameters(
    Index 0 Param 1=1
    Index 0 Param 2=1
    Index 0 Param 4=1
    Module.Index0Float0Write=true
    Module.Index0Float1Write=true
    Module.Index0Float2Write=true
    Module.Index0Float3Write=true
   )
  ]
  rendererClasses [&NiagaraRibbonRendererProperties]
  renderers [{rendererClass=&NiagaraRibbonRendererProperties}]
 }
 Minimal001 {
  emitterUpdateScript [
   &EmitterState()
   &SpawnBurst_Instantaneous(Spawn Count=2, Use Spawn Probability=false, UseLoopCountLimit=false)
  ]
  particleSpawnScript [
   &InitializeParticle(Color=<rgba 0.2856 6.683 10 1>, Lifetime=0.3, UsePositionOffset=false, Write Material Random=true)
   &PartitionParticles()
   &SocketLocation(
    SamplePosOffset=false
    Socket Index=/Niagara/DynamicInputs/Helpers/MaskInt(
     Mask Source:NiagaraInt32=1
     Mask Parameter:NiagaraInt32=$Particles.PartitionParticles.Partition
     Mask:NiagaraInt32=0
     Masked Int:NiagaraInt32=1
     Passthrough Int:NiagaraInt32=2
    )
    Socket Reader=$System.Socket Reader
    Socket Sampling Mode=::NewEnumerator2
    VelocitySpeedLimit=false
   )
  ]
  particleUpdateScript [
   &ParticleState()
   &ScaleColor(
    Scale RGB=&VectorFromFloat(Value=&FloatFromCurve(
      FloatCurve=DataInterface() {propertyValues {Curve {
         defaultValue=1
         keys [
          {interpMode=RCIM_Cubic, tangentMode=RCTM_User, time=0, value=0}
          {interpMode=RCIM_Cubic, tangentMode=RCTM_User, time=0.0415, value=1}
          {tangentMode=RCTM_User, time=1, value=0}
         ]
        }}}
      Scale Curve=&MergeHelper_FloatSwitch(0=1, 1=0.5, 2=0.7, 3=1, Index=&MakeIntFromFloat(Float=$User.BlinkIndex), Num=::NewEnumerator2)
     ))
    ScaleA=true
    ScaleRGB=true
   )
  ]
  rendererClasses [&NiagaraLightRendererProperties]
  renderers [{rendererClass=&NiagaraLightRendererProperties, rendererIndex=0}]
 }
 ParticleDecal001 {
  emitterUpdateScript [
   &EmitterState(Life Cycle Mode=::NewEnumerator1, Loop Behavior=::NewEnumerator1, Loop Duration=1, UseLoopDelay=false)
   &SpawnBurst_Instantaneous(
    Spawn Count=&MergeHelper_IntSwitch(0=9, 1=9, 2=7, 3=9, 4=?, 5=?, Index=&MakeIntFromFloat(Float=$User.BlinkIndex), Num=::NewEnumerator2)
    Use Spawn Probability=false
   )
  ]
  particleSpawnScript [
   &FN_SetSpawnGroup(Number of Spawn Groups:NiagaraInt32=&MergeHelper_IntSwitch(0=9, 1=9, 2=7, 3=9, 4=?, 5=?, Index=&MakeIntFromFloat(Float=$User.BlinkIndex), Num=::NewEnumerator2))
   /Niagara/Modules/Spawn/Initialization/InitializeParticle(
    Color=<rgba 2.6271 19.2283 45 1>
    Lifetime=0.3
    Module.Write Color=true
    Module.Write Lifetime=true
    Module.Write Mass=true
    Module.Write Material Random=true
    Module.Write Position=false
    Module.Write Scale=false
    Module.Write SpriteRotation=false
    Module.Write SpriteSize=true
    Position=&SimulationPosition()
    Sprite Size=&Vector2DFromFloat(Value=50)
   )
  ]
  particleUpdateScript [
   &SkeletalMeshLocation(
    Apply Sampled Orientation to Mesh Particles=true
    Bone / Socket Index=$Particles.SpawnGroup
    Bone Sampling Mode=::NewEnumerator4
    Inherit Velocity=false
    Interpolate Mesh Sampling=false
    MaskSpawnGroup=false
    Sampled Position Offset Coordinate Space:ENiagaraMeshTransforms=::NewEnumerator3
    Skeletal Mesh=DataInterface() {propertyValues {
      DefaultMesh=null
      ExcludeBoneName=root
      OptionalComponentClass=null
      PreviewMesh=&M_Med_Soldier_04_Preview
      SkinningMode=SkinOnTheFly
      SoftSourceActor=null
      SourceMode=Default
      UvSetIndex=0
      WholeMeshLOD=-1
      bExcludeBone=false
      bReadDeformedGeometry=true
      bRequireCurrentFrameData=true
      ComponentTags []
      FilteredBones [spine_03,spine_05,upperarm_l,lowerarm_l,upperarm_r,lowerarm_r,head,calf_l,calf_r]
      FilteredSockets []
      MeshUserParameter {parameter {name=None, typeDefHandle {registeredTypeIndex=1}, varData []}}
      SamplingRegions []
     }}
   )
   &ParticleState()
   &SolveForcesAndVelocity(Clamp Velocity=false, Limit Acceleration=false)
   &ScaleColor(
    Scale Alpha=&Multiply_Float(
     A=1
     B=&FloatFromCurve(FloatCurve=DataInterface() {propertyValues {Curve {keys [<at 0 0>,<at 0.1243 1.0088>,<at 0.3979 1.0088>,<at 1 0>]}}})
    )
    Scale RGB=&Multiply_VectorByFloat(
     Vector=<1 1 1>
     Float=&MergeHelper_FloatSwitch(0=1, 1=0.1, 2=0.4, 3=1, Index=&MakeIntFromFloat(Float=$User.BlinkIndex), Num=::NewEnumerator2)
    )
    ScaleA=true
    ScaleRGB=true
   )
   &CameraOffset(Camera Offset Amount=20)
   &DynamicMaterialParameters(
    Index 0 Param 1=&FloatFromCurve(FloatCurve=DataInterface() {propertyValues {Curve {keys [<at 0 5 RCIM_Cubic>,<at 0.4 1.5 RCIM_Cubic>,<at 1 5>]}}}, Scale Curve=0.8)
    Index 0 Param 2=0.01
    Index 0 Param 3=&FloatFromCurve(FloatCurve=DataInterface() {propertyValues {Curve {keys [<at 0 0.05>,<at 1 0>]}}})
    Index 0 Param 4=&RandomRangeFloat(Minimum=0, Maximum=1)
    Module.Param0WriteEnabled=true
    Module.Param1WriteEnabled=true
    Module.Param2WriteEnabled=true
    Module.Param3WriteEnabled=true
   )
  ]
  rendererClasses [&NiagaraSpriteRendererProperties]
  renderers [{rendererClass=&NiagaraSpriteRendererProperties}]
 }
 Shock_wave {
  emitterUpdateScript [
   &EmitterState(Life Cycle Mode=::NewEnumerator1, Loop Behavior=::NewEnumerator1, Loop Duration=0.5, UseLoopDelay=true)
   &SpawnBurst_Instantaneous(Spawn Count=1, Spawn Time=0.09, Use Spawn Probability=false)
  ]
  particleSpawnScript [
   &InitializeParticle(
    Color=<rgba 1 1 1 0.05>
    Lifetime=0.27
    Mesh Scale=<2 2 2>
    Module.Write Color=true
    Module.Write Lifetime=true
    Module.Write Mass=true
    Module.Write Material Random=true
    Module.Write Position=false
    Module.Write Scale=true
    Module.Write SpriteRotation=false
    Module.Write SpriteSize=true
    Position=&SimulationPosition()
    Sprite Size=<5 5>
   )
   &SystemLocation(MaskSpawnGroup=false, Offset=<-30 0 100>)
   &InitialMeshOrientation(Rotation=<0 0 0.5>, Use Orientation Vector=true, Use Rotation Vector=true)
   /SetVariables(Particles.Mesh_position02:Vector3f=<-40 0 100>) {isSetParameters=true}
  ]
  particleUpdateScript [
   /SetVariables(Particles.Mesh_color02:LinearColor=&MakeLinearColorFromVectorAndFloat(
     "Vector (RGB)"=&Multiply_Vector(
      A=&MakeVectorFromLinearColorRGB(Color=<rgba 0.3296 1.6138 2.5 5>)
      B=&MakeVectorFromLinearColorRGB(Color=&ColorFromCurve(ColorCurve=DataInterface() {propertyValues {
          AlphaCurve {keys []}
          BlueCurve {keys [<at 0 30 RCIM_Cubic>,<at 0.6 0.2 RCIM_Cubic>]}
          GreenCurve {keys [<at 0 20.3007 RCIM_Cubic>,<at 0.6 0.2 RCIM_Cubic>]}
          RedCurve {keys [<at 0 4.0383 RCIM_Cubic>,<at 0.6 0.2 RCIM_Cubic>]}
         }}))
     )
     "Float (Alpha)"=&Multiply_Float(
      A=&FloatFromCurve(FloatCurve=DataInterface() {propertyValues {Curve {keys [<at 0 0 RCIM_Cubic>, {interpMode=RCIM_Cubic, leaveTangent=-1.8775, tangentMode=RCTM_Break, time=0.026, value=1}, <at 0.8 0>]}}})
      B=1
     )
    )) {isSetParameters=true}
   /SetVariables(Particles.Mesh_Size02:Vector3f=&Multiply_Vector(
     A=&MergeHelper_VectorSwitch(
      0=<1.7 1.7 1.7>
      1=<1.4 0.8 0.8>
      2=<1.7 1.7 1.7>
      3=<1.2 1.2 1.2>
      Index=&MakeIntFromFloat(Float=$User.BlinkIndex)
      Num=::NewEnumerator2
     )
     B=&VectorFromCurve(VectorCurve=DataInterface() {propertyValues {XCurve {keys [<at 0 0.5>,<at 0.75 1.5>]}, YCurve {keys [<at 0 0.7>,<at 0.75 1.6>]}, ZCurve {keys [<at 0 0.7>,<at 0.75 2>]}}})
    )) {isSetParameters=true}
   /SetVariables(Particles.Mesh_Dynamic02:Vector4f=&MakeVector4(
     X=&FloatFromCurve(FloatCurve=DataInterface() {propertyValues {Curve {keys [<at 0 0 RCIM_Cubic>,<at 0.8 0.28 RCIM_Cubic>]}}})
     Y=&RandomRangeFloat(Minimum=0, Maximum=1)
     Z=50
     W=0
    )) {isSetParameters=true}
   &ParticleState()
   &SolveForcesAndVelocity(Clamp Velocity=false, Limit Acceleration=false)
   &ScaleColor(
    Scale Alpha=&Multiply_Float(
     A=1
     B=&Multiply_Float(
      A=&FloatFromCurve(FloatCurve=DataInterface() {propertyValues {Curve {keys [<at 0 0 RCIM_Cubic>,<at 0.026 1 RCIM_Cubic>,<at 1 0>]}}})
      B=&FNTimeOfDayScalar(Night Scale=0.5, Morning Scale=0.1, Evening Scale=0.1)
     )
    )
    Scale RGB=&MakeVectorFromLinearColorRGB(Color=&ColorFromCurve(ColorCurve=DataInterface() {propertyValues {
        AlphaCurve {keys []}
        BlueCurve {keys [{arriveTangent=2.3578, interpMode=RCIM_Cubic, time=0, value=6.25}, <at 0.6 1 RCIM_Cubic>]}
        GreenCurve {keys [{arriveTangent=-30.4531, interpMode=RCIM_Cubic, time=0, value=8.6458}, <at 0.6 1 RCIM_Cubic>]}
        RedCurve {keys [{arriveTangent=-115.5308, interpMode=RCIM_Cubic, time=0, value=12}, <at 0.6 1 RCIM_Cubic>]}
       }}))
    ScaleA=true
    ScaleRGB=true
   ) {module=ScaleColor001}
   &ScaleMeshSize(Scale Factor=&VectorFromCurve(
     VectorCurve=DataInterface() {propertyValues {XCurve {keys [<at 0 0.5>,<at 1 1.5>]}, YCurve {keys [<at 0 0.7>,<at 1 1.6>]}, ZCurve {keys [<at 0 0.7>,<at 1 2>]}}}
     Scale Curve=&MergeHelper_VectorSwitch(
      0=<1 1 1>
      1=<1 0.7 0.7>
      2=<1 1 1>
      3=<0.7 0.7 0.7>
      Index=&MakeIntFromFloat(Float=$User.BlinkIndex)
      Num=::NewEnumerator2
     )
    ))
   &DynamicMaterialParameters(
    Index 0 Param 1=&FloatFromCurve(FloatCurve=DataInterface() {propertyValues {Curve {keys [<at 0 0 RCIM_Cubic>,<at 1 0.28 RCIM_Cubic>]}}})
    Index 0 Param 2=&RandomRangeFloat(Minimum=0, Maximum=1)
    Index 0 Param 3=50
    Index 0 Param 4=&FloatFromCurve(FloatCurve=DataInterface() {propertyValues {Curve {keys [
         {interpMode=RCIM_Cubic, leaveTangent=0.0377, tangentMode=RCTM_Break, time=0, value=0}
         {arriveTangent=-3.0374, interpMode=RCIM_Cubic, leaveTangent=2.0371, tangentMode=RCTM_Break, time=0.2, value=-3.5}
         {interpMode=RCIM_Cubic, tangentMode=RCTM_Break, time=1, value=0}
        ]}}})
    Module.Param0WriteEnabled=true
    Module.Param1WriteEnabled=true
    Module.Param2WriteEnabled=true
    Module.Param3WriteEnabled=true
   )
  ]
  rendererClasses [&NiagaraMeshRendererProperties]
  renderers [
   {rendererClass=&NiagaraMeshRendererProperties, rendererIndex=0}
   {rendererClass=&NiagaraMeshRendererProperties, rendererIndex=1}
  ]
 }
}
summary {
 systemName=NS_Emote_LooseCreep
 emitters [
  {emitterName=Shock_wave, rendererClasses [&NiagaraMeshRendererProperties]}
  {emitterName=ParticleDecal001, rendererClasses [&NiagaraSpriteRendererProperties]}
  {emitterName=FN_AnimTrail001, rendererClasses [&NiagaraRibbonRendererProperties]}
  {emitterName=FN_AnimTrail002, rendererClasses [&NiagaraRibbonRendererProperties]}
  {emitterName=Minimal001, rendererClasses [&NiagaraLightRendererProperties]}
 ]
 userVariables [
  User.bIsFrontEnd {type=NiagaraBool}
  User.bIsFrontEndPreview {type=NiagaraBool}
  User.BlinkIndex {type=NiagaraFloat, defaultValue=0, description=""}
  User.BlinkWidth {type=NiagaraFloat, defaultValue=1, description=""}
  User.OwnerLocalVelocity {type=Vector3f, description="", defaultValue {struct=Vector3f, value=<0 0 0>}}
  User.RibbonIDOffset {type=NiagaraInt32}
  User.SK_mesh {
   type=NiagaraDataInterfaceSkeletalMesh
   defaultValue {
    struct=NiagaraExt_VariableValue_DataInterface
    value {dataInterface=&self:NiagaraDataInterfaceSkeletalMesh_3}
   }
  }
 ]
}
