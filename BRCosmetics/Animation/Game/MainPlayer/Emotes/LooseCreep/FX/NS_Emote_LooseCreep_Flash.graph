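# Niagara system for the LooseCreep emote that plays a brief blue flash.
# Two near-identical burst emitters spawn short-lived sprites, one a wide thin streak and one a square, pushed toward the camera.
# Both pop up in size, then shrink while fading out, with dynamic material parameters written to the sprite.
# Spawn position is offset by a system-level FrontEndAdjustment value, alongside exposed front-end and BlinkIndex user parameters.
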NiagaraSystem /BRCosmetics/Animation/Game/MainPlayer/Emotes/LooseCreep/FX/NS_Emote_LooseCreep_Flash
dependencies {
 usedDataInterfaces [&NiagaraDataInterfaceCurve]
 usedDynamicInputs [
  &MakeCustomVectorFromBool
  &Boolean_OrOperation
  &MergeHelper_VectorSwitch
  &MakeIntFromFloat
 ]
 usedModules [
  &SystemState
  &self:SystemSpawnScript.SystemScriptSource.SystemScriptGraph.NiagaraNodeAssignment_0.SetVariables_#1
 ]
}
emitters {
 Minimal {
  emitterUpdateScript [
   &EmitterState()
   &SpawnBurst_Instantaneous(Use Spawn Probability=false, UseLoopCountLimit=false)
  ]
  particleSpawnScript [
   &InitializeParticle(
    Color=<rgba 0.143 0.5135 1 1>
    Lifetime=0.21
    Position Offset=$System.FrontEndAdjustment
    Sprite Size=<200 20>
    Sprite Size Mode=::NewEnumerator4
    UsePositionOffset=true
    Write Material Random=true
   )
   &CameraOffset(Camera Offset Amount=100)
  ]
  particleUpdateScript [
   &ParticleState()
   &ScaleSpriteSize(Uniform Curve Sprite Scale=DataInterface() {propertyValues {Curve {defaultValue=0, keys [<at -0.0001 0>, {tangentMode=RCTM_User, time=0.2, value=1}, {tangentMode=RCTM_User, time=1, value=0}]}}})
   &DynamicMaterialParameters(
    Index 0 Param 1=0.8
    Index 0 Param 2=1
    Index 0 Param 3=1
    Module.Index0Float0Write=true
    Module.Index0Float1Write=true
    Module.Index0Float2Write=true
    Module.Index0Float3Write=true
   )
   &ScaleColor(
    Scale Alpha=&FloatFromCurve(FloatCurve=DataInterface() {propertyValues {Curve {keys [<at 0 1>,<at 1 0>]}}})
    ScaleA=true
    ScaleRGB=true
   )
  ]
  rendererClasses [&NiagaraSpriteRendererProperties]
  renderers [{rendererClass=&NiagaraSpriteRendererProperties}]
 }
 Minimal001 {
  emitterUpdateScript [
   &EmitterState()
   &SpawnBurst_Instantaneous(Use Spawn Probability=false, UseLoopCountLimit=false)
  ]
  particleSpawnScript [
   &InitializeParticle(
    Color=<rgba 0.143 0.5135 1 1>
    Lifetime=0.21
    Position Offset=$System.FrontEndAdjustment
    Sprite Size=<130 130>
    Sprite Size Mode=::NewEnumerator4
    UsePositionOffset=true
    Write Material Random=true
   )
   &CameraOffset(Camera Offset Amount=100)
  ]
  particleUpdateScript [
   &ParticleState()
   &ScaleSpriteSize(Uniform Curve Sprite Scale=DataInterface() {propertyValues {Curve {defaultValue=0, keys [<at -0.0001 0>, {tangentMode=RCTM_User, time=0.2, value=1}, {tangentMode=RCTM_User, time=1, value=0}]}}})
   &DynamicMaterialParameters(
    Index 0 Param 1=0.8
    Index 0 Param 2=1
    Index 0 Param 3=1
    Module.Index0Float0Write=true
    Module.Index0Float1Write=true
    Module.Index0Float2Write=true
    Module.Index0Float3Write=true
   )
   &ScaleColor(
    Scale Alpha=&FloatFromCurve(FloatCurve=DataInterface() {propertyValues {Curve {keys [<at 0 1>,<at 1 0>]}}})
    ScaleA=true
    ScaleRGB=true
   )
  ]
  rendererClasses [&NiagaraSpriteRendererProperties]
  renderers [{rendererClass=&NiagaraSpriteRendererProperties}]
 }
}
summary {
 systemName=NS_Emote_LooseCreep_Flash
 emitters [
  {emitterName=Minimal, rendererClasses [&NiagaraSpriteRendererProperties]}
  {emitterName=Minimal001, rendererClasses [&NiagaraSpriteRendererProperties]}
 ]
 userVariables [
  User.bIsFrontEnd {type=NiagaraBool}
  User.bIsFrontEndPreview {type=NiagaraBool}
  User.BlinkIndex {type=NiagaraFloat, defaultValue=0, description=""}
 ]
}
