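# Niagara system for the landing effect of the LooseCreep emote, made of a dust emitter and a mud splash emitter.
# System-level logic combines SystemState, a variable assignment and a local SetFootDustFXColorArrayIndex module, with a bool-to-float dynamic input also in use.
# User parameters expose a foot direction, a surface type index and a foot dust color array, pointing to surface-dependent dust coloring.
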NiagaraSystem /BRCosmetics/Animation/Game/MainPlayer/Emotes/LooseCreep/FX/NS_Emote_LooseCreep_Landing
dependencies {
 usedDataInterfaces []
 usedDynamicInputs [&MakeCustomFloatFromBool]
 usedModules [
  &self:SetFootDustFXColorArrayIndex
  &self:SystemSpawnScript.SystemScriptSource.SystemScriptGraph.NiagaraNodeAssignment_0.SetVariables_#1
  &SystemState
 ]
}
emitters {LW_Dust {rendererClasses [], renderers []}, LW_Mud_Splash {rendererClasses [], renderers []}}
summary {
 systemName=NS_Emote_LooseCreep_Landing
 emitters [{emitterName=LW_Mud_Splash, rendererClasses []}, {emitterName=LW_Dust, rendererClasses []}]
 userVariables [
  User.FootDirection {type=Vector3f, description="", defaultValue {struct=Vector3f, value=<0 0 0>}}
  User.FootDustColorArray {
   type=NiagaraDataInterfaceArrayColor
   description=""
   defaultValue {
    struct=NiagaraExt_VariableValue_DataInterface
    value {
     dataInterface=&self:NiagaraDataInterfaceArrayColor_0
     dataInterfaceClass=&NiagaraDataInterfaceArrayColor
    }
   }
  }
  User.SurfaceTypeIndex {type=NiagaraInt32}
 ]
}
